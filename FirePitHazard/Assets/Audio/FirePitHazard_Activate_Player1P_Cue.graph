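# Sound cue for the first-person player activation sound of the fire pit hazard, in the traps sound class.
# A random node picks one of four equally weighted launch waves and feeds it through a modulator into a quality-level switch, which fills only its first two slots, both with that chain.
# Concurrency override allows one instance at a time, virtualization is disabled, and attenuation overrides turn off distance falloff and spatialization.
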SoundCue /FirePitHazard/Assets/Audio/FirePitHazard_Activate_Player1P_Cue
root=QualityLevel_0
nodes [
 {id=Modulator_1, children [Random_1]}
 {id=QualityLevel_0, children [Modulator_1,Modulator_1,null,null]}
 {
  id=Random_1
  children [WavePlayer_4,WavePlayer_5,WavePlayer_6,WavePlayer_7]
  properties {bShouldExcludeFromBranchCulling=true, weights [1,1,1,1]}
 }
 {id=WavePlayer_4, properties {soundWaveAssetPtr=&/FirePitHazard_Launch_1P_01}}
 {id=WavePlayer_5, properties {soundWaveAssetPtr=&/FirePitHazard_Launch_1P_02}}
 {id=WavePlayer_6, properties {soundWaveAssetPtr=&/FirePitHazard_Launch_1P_03}}
 {id=WavePlayer_7, properties {soundWaveAssetPtr=&/FirePitHazard_Launch_1P_04}}
]
properties {
 bOverrideConcurrency=true
 maxDistance=2097152
 soundClassObject=&Fort_Traps
 virtualizationMode=Disabled
 volumeMultiplier=1
 attenuationOverrides {bAttenuate=false, bSpatialize=false}
 concurrencyOverrides {maxCount=1}
}
