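# Sound cue for the fire pit hazard's looping ambient audio, routed to the traps sound class with its own attenuation asset.
# A random node picks one of two equally weighted looping waves, feeding a modulator for slight pitch variation and an enveloper that ramps volume up from silence.
# The root quality-level node connects that chain to its first two slots and leaves the other two empty.
# A concurrency override caps simultaneous instances.
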SoundCue /FirePitHazard/Assets/Audio/FirePitHazard_Amb_Loop_Cue
root=QualityLevel_0
nodes [
 {
  id=Enveloper_1
  children [Modulator_0]
  properties {pitchCurve {editorCurveData {keys [{time=0}]}}, volumeCurve {editorCurveData {keys [<at 0 0>,<at 1 1>]}}}
 }
 {id=Modulator_0, children [Random_0], properties {pitchMin=0.98, volumeMax=1, volumeMin=1}}
 {id=QualityLevel_0, children [Enveloper_1,Enveloper_1,null,null]}
 {id=Random_0, children [WavePlayer_1,WavePlayer_2], properties {bShouldExcludeFromBranchCulling=true, weights [1,1]}}
 {id=WavePlayer_1, properties {bLooping=true, soundWaveAssetPtr=&/FirePitHazard_Amb_Loop_01}}
 {id=WavePlayer_2, properties {bLooping=true, soundWaveAssetPtr=&/FirePitHazard_Amb_Loop_02}}
]
properties {
 attenuationSettings=&/FirePitHazard_Amb_Loop_Att
 bOverrideConcurrency=true
 maxDistance=1500
 priority=0
 soundClassObject=&Fort_Traps
 volumeMultiplier=1.239
 attenuationOverrides {attenuationShapeExtents {x=0}}
 concurrencyOverrides {maxCount=3}
}
