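# Sound cue for activating the fire pit hazard's button in the Fort_Traps sound class.
# A single wave player at the root plays the button's idle loop start wave at reduced volume.
# It borrows the button idle attenuation asset for spatial falloff.
# An overridden concurrency rule allows one instance, stopping the lowest priority when exceeded, and the cue's own priority is zero.
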SoundCue /FirePitHazard/Assets/Audio/FirePitHazard_Button_Activate_Cue
root=WavePlayer_0
nodes [{id=WavePlayer_0, properties {soundWaveAssetPtr=&/FirePitHazard_Button_Idle_Loop_Start}}]
properties {
 attenuationSettings=&/FirePitHazard_Button_Idle_Att
 bOverrideConcurrency=true
 maxDistance=1500
 priority=0
 soundClassObject=&Fort_Traps
 volumeMultiplier=0.354
 concurrencyOverrides {maxCount=1, resolutionRule=StopLowestPriority}
}
