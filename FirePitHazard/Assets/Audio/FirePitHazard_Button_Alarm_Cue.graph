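# Sound Cue for the Fire Pit Hazard button alarm, routed to the traps sound class.
# A mixer layers the alarm wave with a louder, pitch-modulated button activation boom that fires after a fixed delay.
# Each layer has its own attenuation asset, the boom using a "Shake" variant.
# A concurrency override limits playback to one instance, stopping the lowest priority, and virtualization is disabled.
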SoundCue /FirePitHazard/Assets/Audio/FirePitHazard_Button_Alarm_Cue
root=Mixer_0
nodes [
 {id=Attenuation_2, children [Delay_0], properties {attenuationSettings=&/FirePitHazard_Button_Alarm_Att}}
 {id=Attenuation_3, children [Delay_2], properties {attenuationSettings=&/FirePitHazard_Button_Alarm_Shake_Att}}
 {id=Delay_0, children [WavePlayer_0]}
 {id=Delay_2, children [Modulator_1], properties {delayMax=1.2, delayMin=1.2}}
 {id=Mixer_0, children [Attenuation_2, Attenuation_3], properties {inputVolume [0.75,1.5]}}
 {id=Modulator_1, children [WavePlayer_3], properties {pitchMax=1.1, volumeMax=1, volumeMin=1}}
 {id=WavePlayer_0, properties {soundWaveAssetPtr=&/FirePitHazard_Button_Alarm}}
 {id=WavePlayer_3, properties {soundWaveAssetPtr=&/FirePitHazard_Button_Press_Activate_Boom}}
]
properties {
 bOverrideConcurrency=true
 maxDistance=8500
 priority=0.9
 soundClassObject=&Fort_Traps
 virtualizationMode=Disabled
 volumeMultiplier=1
 concurrencyOverrides {maxCount=1, resolutionRule=StopLowestPriority}
}
