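# Sound cue that plays the looping idle sound for the fire pit hazard's button, in the traps sound class.
# A quality level node at the root feeds one looping wave player into its first two slots and leaves the other two empty.
# The cue uses a dedicated attenuation asset, reduced volume and zero priority.
# A concurrency override allows only one instance, stopping the lowest priority when exceeded.
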SoundCue /FirePitHazard/Assets/Audio/FirePitHazard_Button_Idle_Cue
root=QualityLevel_0
nodes [
 {id=QualityLevel_0, children [WavePlayer_0,WavePlayer_0,null,null]}
 {id=WavePlayer_0, properties {bLooping=true, soundWaveAssetPtr=&/FirePitHazard_Button_Idle_Loop}}
]
properties {
 attenuationSettings=&/FirePitHazard_Button_Idle_Att
 bOverrideConcurrency=true
 maxDistance=1500
 priority=0
 soundClassObject=&Fort_Traps
 volumeMultiplier=0.354
 concurrencyOverrides {maxCount=1, resolutionRule=StopLowestPriority}
}
