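# Sound cue for the button press sound of the fire pit hazard, classed as a trap sound.
# A single wave player feeds a modulator whose pitch and volume ranges are fixed at unity, so playback has no random variation.
# It borrows the button idle attenuation asset, plays at reduced volume with virtualization disabled, and overrides concurrency to allow one instance, stopping the lowest priority.
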SoundCue /FirePitHazard/Assets/Audio/FirePitHazard_Button_Press_Cue
root=Modulator_0
nodes [
 {id=Modulator_0, children [WavePlayer_0], properties {pitchMax=1, pitchMin=1, volumeMax=1, volumeMin=1}}
 {id=WavePlayer_0, properties {soundWaveAssetPtr=&/FirePitHazard_Button_Press_01}}
]
properties {
 attenuationSettings=&/FirePitHazard_Button_Idle_Att
 bOverrideConcurrency=true
 maxDistance=1500
 priority=0
 soundClassObject=&Fort_Traps
 virtualizationMode=Disabled
 volumeMultiplier=0.496
 concurrencyOverrides {maxCount=1, resolutionRule=StopLowestPriority}
}
