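# Sound cue for the fire pit hazard's eruption buildup, the tell sound that goes with the trap.
# A Random node picks one of three equally weighted Erupt Tell wave variations.
# Playback uses the BuildUp attenuation asset and the Fort_Traps sound class, with boosted volume, virtualization disabled and a concurrency override that caps simultaneous instances.
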SoundCue /FirePitHazard/Assets/Audio/FirePitHazard_Erupt_Buildup_Cue
root=Random_2
nodes [
 {id=Random_2, children [WavePlayer_8, WavePlayer_9, WavePlayer_10], properties {bShouldExcludeFromBranchCulling=true, weights [1,1,1]}}
 {id=WavePlayer_10, properties {soundWaveAssetPtr=&/FirePitHazard_Erupt_Tell_03}}
 {id=WavePlayer_8, properties {soundWaveAssetPtr=&/FirePitHazard_Erupt_Tell_01}}
 {id=WavePlayer_9, properties {soundWaveAssetPtr=&/FirePitHazard_Erupt_Tell_02}}
]
properties {
 attenuationSettings=&/FirePitHazard_Erupt_Burst_BuildUp_Att
 bOverrideConcurrency=true
 maxDistance=2000
 soundClassObject=&Fort_Traps
 virtualizationMode=Disabled
 volumeMultiplier=2
 attenuationOverrides {bEnableListenerFocus=true, falloffDistance=1600, nonFocusVolumeAttenuation=0.5}
 concurrencyOverrides {maxCount=3}
}
