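# Sound cue for the fire pit hazard trap's launch, in the traps sound class with its own attenuation and a concurrency limit.
# An IsEnemy branch picks between two mixes of random launch and eruption variants.
# The true or unset mix favors a briefly, randomly delayed launch over a quieter eruption.
# The false mix plays both without delay and favors the eruption.
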SoundCue /FirePitHazard/Assets/Audio/FirePitHazard_Launch_Cue
root=Modulator_0
nodes [
 {id=Branch_0, children [Mixer_0,Mixer_1,Mixer_0], properties {boolParameterName=IsEnemy}}
 {id=Delay_0, children [Modulator_1], properties {delayMax=0.25, delayMin=0.1}}
 {id=Mixer_0, children [Delay_0,Modulator_3], properties {inputVolume [1,0.5]}}
 {id=Mixer_1, children [Modulator_2,Modulator_4], properties {inputVolume [0.35,1]}}
 {id=Modulator_0, children [Branch_0], properties {volumeMax=1, volumeMin=1}}
 {id=Modulator_1, children [Random_2]}
 {id=Modulator_2, children [Random_3]}
 {id=Modulator_3, children [Random_4]}
 {id=Modulator_4, children [Random_5]}
 {
  id=Random_2
  children [WavePlayer_8, WavePlayer_9, WavePlayer_10, WavePlayer_11]
  properties {bShouldExcludeFromBranchCulling=true, weights [1,1,1,1]}
 }
 {
  id=Random_3
  children [WavePlayer_12, WavePlayer_13, WavePlayer_14, WavePlayer_15]
  properties {bShouldExcludeFromBranchCulling=true, weights [1,1,1,1]}
 }
 {id=Random_4, children [WavePlayer_17, WavePlayer_18, WavePlayer_19, WavePlayer_20, WavePlayer_21], properties {weights [1,1,1,1,1]}}
 {id=Random_5, children [WavePlayer_22, WavePlayer_23, WavePlayer_24, WavePlayer_25, WavePlayer_26], properties {weights [1,1,1,1,1]}}
 {id=WavePlayer_10, properties {soundWaveAssetPtr=&/FirePitHazard_Launch_3P_03}}
 {id=WavePlayer_11, properties {soundWaveAssetPtr=&/FirePitHazard_Launch_3P_04}}
 {id=WavePlayer_12, properties {soundWaveAssetPtr=&/FirePitHazard_Launch_3P_01}}
 {id=WavePlayer_13, properties {soundWaveAssetPtr=&/FirePitHazard_Launch_3P_02}}
 {id=WavePlayer_14, properties {soundWaveAssetPtr=&FirePitHazard_Launch_3P_03}}
 {id=WavePlayer_15, properties {soundWaveAssetPtr=&FirePitHazard_Launch_3P_04}}
 {id=WavePlayer_17, properties {soundWaveAssetPtr=&/FirePitHazard_Erupt_01}}
 {id=WavePlayer_18, properties {soundWaveAssetPtr=&/FirePitHazard_Erupt_02}}
 {id=WavePlayer_19, properties {soundWaveAssetPtr=&/FirePitHazard_Erupt_03}}
 {id=WavePlayer_20, properties {soundWaveAssetPtr=&/FirePitHazard_Erupt_04}}
 {id=WavePlayer_21, properties {soundWaveAssetPtr=&/FirePitHazard_Erupt_05}}
 {id=WavePlayer_22, properties {soundWaveAssetPtr=&FirePitHazard_Erupt_01}}
 {id=WavePlayer_23, properties {soundWaveAssetPtr=&FirePitHazard_Erupt_02}}
 {id=WavePlayer_24, properties {soundWaveAssetPtr=&FirePitHazard_Erupt_03}}
 {id=WavePlayer_25, properties {soundWaveAssetPtr=&FirePitHazard_Erupt_04}}
 {id=WavePlayer_26, properties {soundWaveAssetPtr=&FirePitHazard_Erupt_05}}
 {id=WavePlayer_8, properties {soundWaveAssetPtr=&FirePitHazard_Launch_3P_01}}
 {id=WavePlayer_9, properties {soundWaveAssetPtr=&FirePitHazard_Launch_3P_02}}
]
properties {
 attenuationSettings=&/FirePitHazard_Launch_Att
 bOverrideConcurrency=true
 maxDistance=7500
 soundClassObject=&Fort_Traps
 volumeMultiplier=1.416
 attenuationOverrides {bEnableListenerFocus=true, falloffDistance=2000, attenuationShapeExtents {x=500}}
 concurrencyOverrides {maxCount=3}
}
