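# Translucent, unlit, two-sided material for Niagara mesh particles, forming an animated fireball for the fire pit hazard effect.
# Panning Perlin noise warps a tiled flame texture, and their blend interpolates red to hot orange, tinted by particle color.
# The same mask builds a normal-direction displacement and an eroded, Fresnel-shaped alpha scaled by particle alpha.
# Dynamic parameters control flame pan, erosion, displacement intensity and the wave width of a separate streak-texture lookup.
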Material /FirePitHazard/Assets/VFX/M_FirePitHazard_Ball1
graph {
 properties {
  bAutomaticallySetUsageInEditor=false
  bUsedWithNiagaraMeshParticles=true
  blendMode=BLEND_Translucent
  pixelDepthOffsetMode=PDOM_Legacy
  shadingModel=MSM_Unlit
  twoSided=true
  usedShadingModels=Unlit
 }
 nodes [
  n1 = Constant3Vector() {constant=<rgba 1 0.0167 0 1>}
  n2 = Constant3Vector() {constant=<rgba 2 1.3431 0.1617 1>}
  n3 = TextureCoordinate() {uTiling=1.5, vTiling=1.5}
  n4 = TextureSampleParameter2D(UVs=Panner(Coordinate=Multiply(A=@n3)) {speedX=0.1, speedY=-0.07}) {parameterName=WarpNoise, samplerType=SAMPLERTYPE_LinearColor, texture=/Game/Effects/SnowSandInteraction/Textures/SSI_Perlin}
  n5 = TextureCoordinate() {uTiling=3, vTiling=2}
  n6 = MaterialFunctionCall(Float2=@n5) {materialFunction=&BreakOutFloat2Components}
  n7 = DynamicParameter() {defaultValue=<rgba 0 0 0 1>, paramNames [PanFlame,Erode,WPOIntensity,PanWaveWidth]}
  n8 = Add(A=@n6.R, B=@n7.PanFlame)
  n9 = TextureSampleParameter2D(UVs=Add(A=Multiply(A=@n4.R) {constB=0.2}, B=AppendVector(A=@n8, B=@n6.G))) {
   parameterName=TexFlame
   samplerType=SAMPLERTYPE_LinearGrayscale
   texture=/Game/Effects/Fort_Effects/Emotes/UltraLight/Texture/T_FluidFlameTile_Greyscale
  }
  n10 = Multiply(A=LinearInterpolate(A=@n4.R, B=@n9.R))
  n11 = Power(Base=@n10) {constExponent=1.4}
  n12 = ParticleColor()
  MaterialFunctionCall(Emissive=Multiply(A=LinearInterpolate(A=@n1, B=@n2, Alpha=@n11), B=@n12.RGB)) {materialFunction=&MF_FX_EmissiveAdjustments}
  n13 = LinearInterpolate(Alpha=@n10) {constA=0.2}
  Multiply(A=Multiply(A=Saturate(None=@n13), B=VertexNormalWS()), B=@n7.WPOIntensity) {constB=30}
  n14 = Fresnel() {baseReflectFraction=-0.1, exponent=0.25}
  n15 = Subtract(A=@n10, B=LinearInterpolate(Alpha=@n7.Erode) {constA=-1})
  Multiply(A=Power(Base=OneMinus(None=@n14)), B=Multiply(A=@n12.A, B=Multiply(A=@n15, B=@n10)))
  Panner(Coordinate=@n5) {speedX=0.3, speedY=0.1}
  n16 = MaterialFunctionCall(Float2=TextureCoordinate()) {materialFunction=&BreakOutFloat2Components}
  n17 = Add(A=@n7.PanFlame, B=Multiply(A=@n16.R, B=@n7.PanWaveWidth) {constB=0.3})
  TextureSampleParameter2D(UVs=Saturate(None=AppendVector(A=@n16.G, B=@n17))) {parameterName=Tex1, texture=/Game/Effects/Fort_Effects/Textures/Smoke/ShockwaveStreak_1}
 ]
}
