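# Translucent unlit sprite material for the fire pit hazard's fireball VFX, driven per particle through four sets of dynamic parameters.
# A motion-vector flipbook is sampled with flowmap-swirled UVs.
# Detail noise indexes a curve atlas for boosted fake-blackbody fire emissive, combined with particle-tinted color, alongside time-of-day smoke colors.
# Opacity combines particle alpha, depth fade, noise-erosion dissolve and a sine-based horizontal edge mask.
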Material /FirePitHazard/Assets/VFX/M_FirePitHazard_FireBall
graph {
 properties {
  bUsedWithNanite=true
  bUsedWithNiagaraSprites=true
  bUsedWithParticleSprites=true
  blendMode=BLEND_Translucent
  pixelDepthOffsetMode=PDOM_Legacy
  shadingModel=MSM_Unlit
  translucencyLightingMode=TLM_VolumetricPerVertexDirectional
  translucentSelfShadowDensityScale=1
  usedShadingModels=Unlit
  translucentMultipleScatteringExtinction=<rgba 1 1 1 1>
 }
 nodes [
  n1 = NamedRerouteUsage()
  n2 = Multiply(A=NamedRerouteUsage(), B=ParticleColor().RGB)
  n3 = NamedRerouteUsage()
  n4 = DynamicParameter() {parameterIndex=1, defaultValue=<rgba 1 800 2 0.1>, paramNames [Temperature, Fake BlackBody Intensity, Alpha Power, Detail UV Ofs]}
  n5 = ScalarParameter() {defaultValue=1.2, parameterName=DissolveNoise Scale}
  n6 = DynamicParameter() {
   parameterIndex=2
   defaultValue=<rgba 0.5 0 0.5 1>
   paramNames [Smoke - DissolveLerpNoise, Smoke - DissolveNoiseOffsetX, Smoke - DissolveNoiseErosionPower, Smoke - DissolveNoiseScale]
  }
  n7 = Add(A=Multiply(A=NamedRerouteUsage(), B=Multiply(A=@n5, B=@n6.Smoke - DissolveNoiseScale)) {desc=Scale}, B=@n6.Smoke - DissolveNoiseOffsetX)
  n8 = DynamicParameter() {parameterIndex=1, defaultValue=<rgba 1 800 2 0.1>, paramNames [Temperature, Fake BlackBody Intensity, Alpha Power, Detail UV Ofs]}
  n9 = TextureSample(UVs=Add(A=@n7, B=@n8.Detail UV Ofs) {constB=0.1}) {texture=/Game/Effects/Fort_Effects/Textures/Fire/T_FireNoise_02}
  n10 = StaticSwitch(True=OneMinus(None=@n9.R), False=@n9.R, Value=StaticBoolParameter() {group=Fire, parameterName=Inverse Detail Tex})
  n11 = DynamicParameter() {parameterIndex=3, defaultValue=<rgba 6 0 0.0006 1>, paramNames [FireNoisePow, Swirl UV Pan W, Swirl UV Scale, Depth Fade Scale]}
  n12 = CurveAtlasRowParameter(CurveTime=Saturate(None=Power(Base=@n10, Exp=@n11.FireNoisePow))) {
   atlas=/Game/Effects/Niagara/MasterMaterials/Smoke/Textures/CurveAtlas
   curve=/Game/Effects/Niagara/MasterMaterials/Smoke/Textures/EmissiveCurveFire
   defaultValue=4
   parameterName=EmissiveCurve
  }
  n13 = Multiply(A=@n12, B=@n12)
  n14 = Multiply(A=@n13, B=@n13)
  n15 = Multiply(A=Max(B=@n4.Temperature), B=Add(A=Multiply(A=@n4.Fake BlackBody Intensity, B=@n14), B=@n12))
  MaterialFunctionCall(Emissive=Add(A=Multiply(A=@n1, B=@n2), B=Multiply(A=@n3, B=@n15))) {materialFunction=&MF_FX_EmissiveAdjustments}
  n16 = DynamicParameter() {defaultValue=<rgba 0 0.03 2 1>, paramNames [Animation, Swirl Strength, Emissive Scale, Sphere Normal Scale]}
  Multiply(A=ScalarParameter() {defaultValue=64, group=Fire, parameterName=HDR_Intensity}, B=@n16.Emissive Scale)
  n17 = DynamicParameter() {defaultValue=<rgba 0 0.03 2 1>, paramNames [Animation, Swirl Strength, Emissive Scale, Sphere Normal Scale]}
  n18 = TextureCoordinate()
  n19 = DynamicParameter() {parameterIndex=3, defaultValue=<rgba 6 0 0.0006 1>, paramNames [FireNoisePow, Swirl UV Pan W, Swirl UV Scale, Depth Fade Scale]}
  n20 = AppendVector(A=ParticleRandom(), B=@n19.Swirl UV Pan W)
  n21 = TextureCoordinate()
  n22 = TextureSample(UVs=Multiply(A=Multiply(A=@n19.Swirl UV Scale) {constB=1000}, B=Add(A=@n20, B=@n21))) {samplerType=SAMPLERTYPE_LinearColor, texture=/Game/VisualThreatManager/StormVisuals/LargeClouds/T_Swirl_Flowmap}
  n23 = Multiply(A=Subtract(A=ComponentMask(None=@n22.RGB) {g=true, r=true}) {constB=0.5}, B=@n17.Swirl Strength)
  n24 = TextureObjectParameter() {parameterName=MotionVector, samplerType=SAMPLERTYPE_LinearColor, texture=/Game/Effects/Niagara/MasterMaterials/Smoke/Textures/Vector_OF}
  n25 = MaterialFunctionCall!(
   "Animation  Phase (0-1)"=@n17.Animation
   Number of Columns X=ScalarParameter() {defaultValue=4, parameterName=Repeat U}
   Number of Rows Y=ScalarParameter() {defaultValue=4, parameterName=Repeat V}
   Texture=TextureObject() {samplerType=SAMPLERTYPE_LinearColor, texture=/Game/Effects/Explosion_Update/Textures/RT4x4_Combined_v2}
   UVs=NamedRerouteDeclaration(None=Add(A=@n18, B=@n23)) {name=Swirl UV, nodeColor=<rgba 0 0.4141 1 1>}
   Motion Vector Texture=@n24
   Phase to Debug
   Motion Vector Intensity=ScalarParameter() {defaultValue=0.344, group=Motion Vectors, parameterName=MotionVector}
   Double Apply Motion Vectors
   Custom UV Input
   Custom UV2 Input
   Use Custom UVs
  ) {materialFunction=/Engine/Functions/Engine_MaterialFunctions02/Texturing/FlipBook_MotionVectors}
  NamedRerouteDeclaration(None=@n25.Result) {name=Flipbook_RGB, nodeColor=<rgba 1 0.4453 0 1>}
  NamedRerouteDeclaration(None=@n25.Alpha) {name=Flipbook_A, nodeColor=<rgba 0 0.7187 1 1>}
  n26 = MaterialFunctionCall(
   Color_Morning=VectorParameter() {group=Time of Day Color, parameterName=Color Morning, defaultValue=<rgba 0.4375 0.3157 0.2716 1>}.RGB
   Color_Day=VectorParameter() {group=Time of Day Color, parameterName=Color Day, defaultValue=<rgba 0.2708 0.2275 0.1823 1>}.RGB
   Color_Evening=VectorParameter() {group=Time of Day Color, parameterName=Color Evening, defaultValue=<rgba 0.5208 0.2454 0.1356 1>}.RGB
   Color_Night=VectorParameter() {group=Time of Day Color, parameterName=Color Night, defaultValue=<rgba 0.131 0.1262 0.151 1>}.RGB
  ) {materialFunction=&Color_Time_of_Day}
  n27 = MaterialFunctionCall(
   Color_Morning=VectorParameter() {group=Time of Day Color, parameterName=Color Morning, defaultValue=<rgba 0.4375 0.3157 0.2716 1>}.RGB
   Color_Day=VectorParameter() {group=Time of Day Color, parameterName=Color Day, defaultValue=<rgba 0.2708 0.2275 0.1823 1>}.RGB
   Color_Evening=VectorParameter() {group=Time of Day Color, parameterName=Color Evening, defaultValue=<rgba 0.5208 0.2454 0.1356 1>}.RGB
   Color_Night=VectorParameter() {group=Time of Day Color, parameterName=Color Night, defaultValue=<rgba 0.131 0.1262 0.151 1>}.RGB
  ) {materialFunction=&Color_Time_of_Day}
  NamedRerouteDeclaration(None=FeatureLevelSwitch(Default=@n26.Result, ES2=@n27.Result, ES3_1=@n27.Result, SM5=@n26.Result, SM6=@n26.Result)) {name=ToD_Smoke, nodeColor=<rgba 1 0 0.2578 1>}
  n28 = VectorParameter() {parameterName=ToD Temperature, defaultValue=<rgba 0.7 1 0.6 0.45>}
  NamedRerouteDeclaration(None=MaterialFunctionCall(Value_Morning=@n28.R, Value_Day=@n28.R, Value_Evening=@n28.R, Value_Night=@n28.R) {materialFunction=&Scalar_Time_of_Day}.Result) {name=ToD Temperature, nodeColor=<rgba 0.9922 1 0 1>}
  n29 = NamedRerouteUsage()
  n30 = Constant() {r=1}
  n31 = DynamicParameter() {parameterIndex=3, defaultValue=<rgba 6 0 0.0006 1>, paramNames [FireNoisePow, Swirl UV Pan W, Swirl UV Scale, Depth Fade Scale]}
  n32 = LinearInterpolate(
   A=Multiply(A=@n31.Depth Fade Scale, B=ScalarParameter() {defaultValue=100, parameterName=DepthFade_Low})
   B=Multiply(A=@n31.Depth Fade Scale, B=ScalarParameter() {defaultValue=10, parameterName=DepthFade_high})
   Alpha=NamedRerouteUsage()
  )
  n33 = Multiply(A=ParticleColor().A, B=Multiply(A=ComponentMask(None=@n29) {g=true}, B=DepthFade(Opacity=@n30, FadeDistance=@n32)))
  n34 = TextureObjectParameter() {
   parameterName=Dissolve Detail Texture
   samplerType=SAMPLERTYPE_LinearColor
   texture=/Game/Effects/Fort_Effects/Effects/Environment/Desert/Dust/Organize/T_Turbulence_Seamless
  }
  n35 = Power(Base=TextureSample(UVs=@n7, Tex=@n34).R, Exp=@n6.Smoke - DissolveNoiseErosionPower)
  n36 = DynamicParameter() {
   parameterIndex=2
   defaultValue=<rgba 0.5 0 0.5 1>
   paramNames [Smoke - DissolveLerpNoise, Smoke - DissolveNoiseOffsetX, Smoke - DissolveNoiseErosionPower, Smoke - DissolveNoiseScale]
  }
  n37 = LinearInterpolate(A=@n33, B=Multiply(A=@n33, B=Saturate(None=@n35)), Alpha=Saturate(None=@n36.Smoke - DissolveLerpNoise))
  n38 = DynamicParameter() {parameterIndex=1, defaultValue=<rgba 1 800 2 0.1>, paramNames [Temperature, Fake BlackBody Intensity, Alpha Power, Detail UV Ofs]}
  n39 = Multiply(A=TextureCoordinate()) {constB=0.5}
  n40 = MaterialFunctionCall(Float2=Abs(None=Sine(None=@n39))) {materialFunction=&BreakOutFloat2Components}
  n41 = Saturate(None=Multiply(A=Multiply(A=@n40.R, B=@n40.R)) {constB=2})
  NamedRerouteDeclaration(None=Multiply(A=Power(Base=@n37, Exp=@n38.Alpha Power), B=Multiply(A=@n41, B=@n41))) {name=Opacity_Result, nodeColor=<rgba 0.1016 1 0 1>}
 ]
}
