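# Unlit, two-sided translucent material for fire sprites on Niagara mesh particles, named for the fire pit hazard.
# A Sprite function builds world position offset for quads, sized by scale parameters, a dynamic Quad Scale and vertex-red variation.
# An 8x8 smoke flipbook, phase-offset by vertex green and animated over time, is tinted by Fire Color for emissive and opacity.
# Opacity uses depth fade, alongside a squared near-camera fade.
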Material /FirePitHazard/Assets/VFX/M_FirePitHazard_FireSprite
graph {
 properties {
  bUsedWithNiagaraMeshParticles=true
  blendMode=BLEND_Translucent
  opacityMaskClipValue=0.05
  pixelDepthOffsetMode=PDOM_Legacy
  shadingModel=MSM_Unlit
  twoSided=true
  usedShadingModels=Unlit
 }
 nodes [
  n1 = WorldPosition()
  n2 = ObjectPositionWS()
  n3 = VertexColor()
  n4 = NamedRerouteDeclaration(None=Add(A=Multiply(A=@n3.R) {constB=0.7}) {constB=0.3}) {name=Noise Channel Red, nodeColor=<rgba 0 1 0.4844 1>}
  n5 = DynamicParameter() {defaultValue=<rgba 1 1500 0 1>, paramNames [Quad Scale, Anim Scale, Anim Speed, Fake Gravity]}
  n6 = AppendVector(
   A=Multiply(A=ScalarParameter() {defaultValue=35, parameterName=ScaleX}, B=@n5.Quad Scale)
   B=Multiply(A=ScalarParameter() {defaultValue=45, parameterName=ScaleY}, B=@n5.Quad Scale)
  )
  Add(
   A=Transform(None=Subtract(A=@n1.XYZ, B=@n2)) {transformSourceType=TRANSFORMSOURCE_World, transformType=TRANSFORM_Local}
   B=MaterialFunctionCall!(Center Location, 0-1 UVs, XY scale=Multiply(A=@n4, B=@n6), OverridePivot, "Normals (Optional)") {materialFunction=&Sprite}.World position offset
  )
  n7 = VectorParameter() {parameterName=Fire Color, defaultValue=<rgba 8 0.8454 0 1>}
  n8 = VertexColor()
  n9 = Add(A=@n8.G, B=Multiply(A=Time(), B=ScalarParameter() {defaultValue=1.5, parameterName=Speed}))
  n10 = Constant() {r=8}
  n11 = TextureObjectParameter() {parameterName=Tex, samplerType=SAMPLERTYPE_LinearColor, texture=/Game/Effects/Fort_Effects/Textures/Smoke/T_Inky_Smoke_SubUV_8x8}
  n12 = StaticComponentMaskParameter(None=MaterialFunctionCall("Animation  Phase (0-1)"=Frac(None=@n9), Number of Rows=@n10, Number of Columns=@n10, Texture=@n11) {materialFunction=&FlipBook}.Result) {defaultR=true, parameterName=Tex Ch}
  MaterialFunctionCall(Emissive=Multiply(A=Multiply(A=@n7.RGB, B=@n12), B=NamedRerouteUsage())) {materialFunction=&MF_FX_EmissiveAdjustments}
  Multiply(A=@n8.G) {constB=0.1}
  n13 = Multiply(A=ComponentMask(None=@n12) {r=true}, B=ScalarParameter() {defaultValue=1.1, parameterName=Alpha})
  n14 = ScalarParameter() {defaultValue=5, parameterName=Depth Fade}
  Multiply(A=Multiply(A=@n7.A, B=DepthFade(Opacity=@n13, FadeDistance=@n14) {fadeDistanceDefault=20}), B=NamedRerouteUsage())
  n15 = MaterialFunctionCall(Distance=ScalarParameter() {defaultValue=100, parameterName=Near Camera Fade Distance}) {materialFunction=&Near_Camera_Fade}
  n16 = Multiply(A=@n15.Result, B=@n15.Result)
  NamedRerouteDeclaration(None=@n16) {name=Near Camera Fade, nodeColor=<rgba 0 0.4844 1 1>}
  ParticlePositionWS()
  VertexInterpolator(VS=@n16)
 ]
}
