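# Translucent, unlit, two-sided sprite material for the fire pit hazard VFX, usable on Niagara mesh particles, with a tinted orange emissive glow.
# A sprite function with X/Y scale and a camera offset function supply position offset math.
# A radial sphere mask, reshaped by power and sine, drives opacity through two depth fades, the first with mask-dependent distance.
# A time sine pulses opacity, and a squared near-camera fade attenuates it.
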Material /FirePitHazard/Assets/VFX/M_FirePitHazard_Sprite
graph {
 properties {
  bUsedWithNiagaraMeshParticles=true
  blendMode=BLEND_Translucent
  opacityMaskClipValue=0.05
  pixelDepthOffsetMode=PDOM_Legacy
  shadingModel=MSM_Unlit
  twoSided=true
  usedShadingModels=Unlit
 }
 nodes [
  n1 = Subtract(A=WorldPosition().XYZ, B=ObjectPositionWS())
  n2 = ScalarParameter() {defaultValue=310, parameterName=ScaleX}
  n3 = ScalarParameter() {defaultValue=190, parameterName=ScaleY}
  Add(
   A=Transform(None=@n1) {transformSourceType=TRANSFORMSOURCE_World, transformType=TRANSFORM_Local}
   B=MaterialFunctionCall!(Center Location, 0-1 UVs, XY scale=AppendVector(A=@n2, B=@n3), OverridePivot, "Normals (Optional)") {materialFunction=&Sprite}.World position offset
  )
  MaterialFunctionCall(Offset Amount=ScalarParameter() {defaultValue=100, parameterName=CamOfs}, World Position=@n1) {materialFunction=/Engine/Functions/Engine_MaterialFunctions02/WorldPositionOffset/CameraOffset}
  n4 = VectorParameter() {parameterName=Fire Color, defaultValue=<rgba 4 0.4227 0 0.1>}
  MaterialFunctionCall(Emissive=@n4.RGB) {materialFunction=&MF_FX_EmissiveAdjustments}
  n5 = MaterialFunctionCall() {materialFunction=&DebugTimeSine}
  n6 = Constant() {r=0.5}
  n7 = SphereMask(A=TextureCoordinate(), B=@n6, Radius=@n6, Hardness=Constant())
  n8 = ScalarParameter() {defaultValue=2, parameterName=Opacity Exponent}
  n9 = Add(A=Multiply(A=Power(Base=@n7, Exp=@n8)) {constB=0.5}) {constB=-0.25}
  n10 = Saturate(None=ConstantBiasScale(None=Sine(None=@n9)))
  n11 = DepthFade(
   Opacity=@n10
   FadeDistance=LinearInterpolate(
    A=ScalarParameter() {defaultValue=100, parameterName=DepthFade_Low}
    B=ScalarParameter() {defaultValue=10, parameterName=DepthFade_high}
    Alpha=@n10
   )
  )
  n12 = DepthFade(
   Opacity=Multiply(A=ComponentMask(None=@n11) {r=true}, B=ScalarParameter() {defaultValue=1.1, parameterName=Alpha})
   FadeDistance=ScalarParameter() {defaultValue=5, parameterName=Depth Fade}
  ) {fadeDistanceDefault=20}
  Multiply(A=Add(A=Multiply(A=@n5.Result) {constB=0.4}) {constB=0.6}, B=Multiply(A=Multiply(A=@n4.A, B=@n12), B=NamedRerouteUsage()))
  n13 = MaterialFunctionCall(Distance=ScalarParameter() {defaultValue=100, parameterName=Near Camera Fade Distance}) {materialFunction=&Near_Camera_Fade}
  NamedRerouteDeclaration(None=Multiply(A=@n13.Result, B=@n13.Result)) {name=Near Camera Fade, nodeColor=<rgba 0 0.4844 1 1>}
 ]
}
